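# Default lit material for the Festivus motorized light fixture, usable on skeletal and Nanite meshes, with diffuse, normal and mask textures.
# A UV1 light pattern, masked and tinted red, is hue-shifted and scaled by brightness with optional beat modulation, all driven by Lighting_MPC.
# A flipbook chain derives atlas UVs from a frame number or FFT-driven range, but the pattern sample does not use it.
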Material /Game/Athena/Environments/Festivus/MotorizedLights/M_MotorizedLight_01
graph {
 properties {
  bUsedWithMeshDeformer=true
  bUsedWithNanite=true
  bUsedWithSkeletalMesh=true
  numCustomizedUVs=2
  pixelDepthOffsetMode=PDOM_Legacy
  usedShadingModels=Default Lit
 }
 nodes [
  n1 = TextureCoordinate() {coordinateIndex=1}
  n2 = ComponentMask(None=@n1) {r=true}
  n3 = ScalarParameter() {defaultValue=64, desc=Rows, parameterName=Rows}
  n4 = Divide(A=Constant() {r=1}, B=@n3)
  n5 = ComponentMask(None=@n1) {g=true}
  n6 = CollectionParameter() {collection=&/Lighting_MPC, parameterName=5x5_FFT_LowNumber}
  n7 = CollectionParameter() {collection=&Lighting_MPC, parameterName=5x5_FFT_HighNumber}
  n8 = Clamp(None=CollectionParameter() {collection=/Game/Athena/Environments/Festivus/Materials/FFT_Materials/FFT_Breakout_MCP, parameterName=FFT_01_Smothed})
  n9 = Floor(None=LinearInterpolate(
    A=CollectionParameter() {collection=&Lighting_MPC, parameterName=5x5_FrameNumber}
    B=LinearInterpolate(A=@n6, B=@n7, Alpha=@n8) {constA=16, constB=21}
    Alpha=CollectionParameter() {collection=&Lighting_MPC, parameterName=5x5_UseFFT}
   ))
  n10 = Fmod(A=@n9, B=@n3)
  n11 = Floor(None=Divide(A=@n9, B=@n3))
  Add(A=AppendVector(A=Multiply(A=@n2, B=@n4), B=Multiply(A=@n5, B=@n4)), B=AppendVector(A=Multiply(A=@n10, B=@n4), B=Multiply(A=@n11, B=@n4)))
  ComponentMask(None=CollectionParameter() {collection=&Lighting_MPC, parameterName=SpotStrip_01_Color}) {b=true, g=true, r=true}
  MaterialFunctionCall!(FramesX, FramesY, Fixed Z, SpriteWidth, Position, SpriteHeight, Rotation, Rotation For Normals, Normals) {materialFunction=/Engine/ArtTools/RenderToTexture/MaterialFunctions/ImposterUVs}
  n12 = Constant3Vector() {constant=<rgba 1 0 0 0>}
  n13 = Multiply(
   A=TextureSample(UVs=TextureCoordinate() {coordinateIndex=1}) {texture=&/T_LightPatterns_01_M}.RGB
   B=TextureSample() {texture=&/T_MotorizedLight_01_M}.G
  )
  n14 = CollectionParameter() {collection=&Lighting_MPC, parameterName=5x5_Brightness}
  n15 = Multiply(A=@n14, B=CollectionParameter() {collection=&Lighting_MPC, parameterName=1_Beat_1})
  n16 = CollectionParameter() {collection=&Lighting_MPC, parameterName=5x5_UseBeat1}
  Multiply(
   A=MaterialFunctionCall(
    Hue Shift Percentage=CollectionParameter() {collection=&Lighting_MPC, parameterName=5x5_HueShift}
    Texture=Multiply(A=@n12, B=@n13) {constB=10}
   ) {materialFunction=&HueShift}.Result
   B=Multiply(A=LinearInterpolate(A=@n14, B=@n15, Alpha=@n16)) {constB=150}
  ) {constB=150}
  ScalarParameter() {desc=FrameNumber, parameterName=FrameNumber}
  ScalarParameter() {parameterName=HueShift}
  ScalarParameter() {defaultValue=1, parameterName=Brightness}
  Subtract(A=@n9)
  TextureSample() {samplerType=SAMPLERTYPE_Normal, texture=&/T_MotorizedLight_01_N}
  TextureSample() {texture=&/T_MotorizedLight_01_D}
  VectorParameter() {parameterName=Color, defaultValue=<rgba 0.1284 0.7889 1 1>}
 ]
}
