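# Material function for Festivus motorized spotlights that outputs an animated four-color emission pattern from a communication timeline texture.
# Time since GameTime, less Delay and divided by Duration, is clamped to sample the timeline once without looping.
# Each RGBA channel drives its own tint, and saturated local-Y offsets blend them into four hard-edged bands along the mesh.
# The four Brightness parameters are declared but unconnected.
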MaterialFunction /Game/Athena/Environments/Festivus/MotorizedLights/MF_Sptilight_Commuication
graph {nodes [
  n1 = Subtract(A=Subtract(A=Time(), B=ScalarParameter() {parameterName=GameTime}), B=ScalarParameter() {defaultValue=1, parameterName=Delay})
  n2 = ScalarParameter() {defaultValue=4.31, parameterName=Duration}
  n3 = TextureSample(UVs=Saturate(None=Divide(A=@n1, B=@n2))) {texture=&/T_Communication_Timeline}
  n4 = ComponentMask(None=MaterialFunctionCall() {materialFunction=&LocalPosition}.Local Position) {g=true}
  n5 = ScalarParameter() {defaultValue=140, parameterName=MOffset1}
  n6 = LinearInterpolate(
   A=Multiply(A=@n3.A, B=VectorParameter() {parameterName=Color1, defaultValue=<rgba 1 0 0 1>}.RGB)
   B=Multiply(A=@n3.B, B=VectorParameter() {parameterName=Color2, defaultValue=<rgba 0 1 0 1>}.RGB)
   Alpha=Saturate(None=Add(A=@n4, B=@n5))
  )
  n7 = Multiply(A=@n3.G, B=VectorParameter() {parameterName=Color3, defaultValue=<rgba 0 0 1 1>}.RGB)
  n8 = Saturate(None=@n4)
  n9 = VectorParameter() {parameterName=Color4, defaultValue=<rgba 1 1 0 1>}
  n10 = Add(A=@n4, B=ScalarParameter() {defaultValue=-140, parameterName=MOffset3})
  FunctionOutput(None=LinearInterpolate(A=LinearInterpolate(A=@n6, B=@n7, Alpha=@n8), B=Multiply(A=@n9.RGB, B=@n3.R), Alpha=Saturate(None=@n10)))
  ScalarParameter() {defaultValue=1, parameterName=Brightness1}
  ScalarParameter() {defaultValue=1, parameterName=Brightness2}
  ScalarParameter() {defaultValue=1, parameterName=Brightness3}
  ScalarParameter() {defaultValue=1, parameterName=Brightness4}
  TextureCoordinate()
 ]}
