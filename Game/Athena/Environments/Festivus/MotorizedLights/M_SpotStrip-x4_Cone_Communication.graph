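# Additive, unlit, two-sided material for the light cones of the motorized four-spot strip in the Festivus environment.
# A communication function's output is multiplied by two panning world-space noises plus a light-pattern mask, then shaped by inverse Fresnel and vertex color.
# Lighting MPC values drive a frame-row UV offset and a cone-size vector built from vertex normals, weighted by vertex alpha.
# Separate nodes blend Color with a custom-primitive-data override.
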Material /Game/Athena/Environments/Festivus/MotorizedLights/M_SpotStrip-x4_Cone_Communication
graph {
 properties {
  bUsedWithMeshDeformer=true
  bUsedWithSkeletalMesh=true
  blendMode=BLEND_Additive
  numCustomizedUVs=2
  pixelDepthOffsetMode=PDOM_Legacy
  shadingModel=MSM_Unlit
  twoSided=true
  usedShadingModels=Unlit
 }
 nodes [
  n1 = ScalarParameter() {bUseCustomPrimitiveData=true, parameterName=bUseCustomColor, primitiveDataIndex=27}
  n2 = VectorParameter() {bUseCustomPrimitiveData=true, parameterName=CustomColor, primitiveDataIndex=28, defaultValue=<rgba 1 0 0 0>}
  Add(A=Multiply(A=@n1, B=@n2.RGB), B=Multiply(A=OneMinus(None=@n1), B=VectorParameter() {parameterName=Color, defaultValue=<rgba 1 0 0 0>}.RGB))
  n3 = TextureCoordinate() {coordinateIndex=1}
  n4 = ComponentMask(None=@n3) {g=true}
  n5 = Divide(A=Constant() {r=1}, B=ScalarParameter() {defaultValue=512, desc=Rows, parameterName=Rows})
  n6 = Floor(None=CollectionParameter() {collection=&/Lighting_MPC, parameterName=SpotStrip_01_FrameNumber})
  AppendVector(A=ComponentMask(None=@n3) {r=true}, B=Add(A=Multiply(A=@n4, B=@n5), B=Multiply(A=@n5, B=@n6)))
  n7 = Constant() {r=1024}
  n8 = Divide(A=WorldPosition() {worldPositionShaderOffset=WPT_ExcludeAllShaderOffsets}.XYZ, B=Multiply(A=Abs(None=@n7)) {constB=-1})
  n9 = AppendVector(A=ComponentMask(None=@n8) {r=true}, B=ComponentMask(None=@n8) {g=true})
  n10 = Time()
  n11 = AppendVector(A=Multiply(A=@n10) {constB=0.1}, B=Multiply(A=@n10) {constB=0.09})
  n12 = Constant() {r=1460}
  n13 = Divide(A=WorldPosition() {worldPositionShaderOffset=WPT_ExcludeAllShaderOffsets}.XYZ, B=Multiply(A=Abs(None=@n12)) {constB=-1})
  n14 = AppendVector(A=ComponentMask(None=@n13) {r=true}, B=ComponentMask(None=@n13) {g=true})
  n15 = Time()
  n16 = AppendVector(A=Multiply(A=@n15) {constB=-0.13}, B=Multiply(A=@n15) {constB=0.127})
  n17 = Multiply(
   A=TextureSample(UVs=Add(A=@n9, B=@n11)) {samplerSource=SSM_Wrap_WorldGroupSettings, texture=/Game/Global/Textures/Default/Noise/T_Tiling_Noise_01}.RGB
   B=TextureSample(UVs=Add(A=@n14, B=@n16)) {samplerSource=SSM_Wrap_WorldGroupSettings, texture=&T_Tiling_Noise_01}.RGB
  )
  n18 = VertexColor()
  n19 = Multiply(A=@n18.A, B=@n18.A)
  n20 = TextureSample(UVs=TextureCoordinate() {coordinateIndex=1}) {texture=&/T_Spot_x4_LightPatterns_01_M}
  n21 = Multiply(
   A=MaterialFunctionCall() {materialFunction=&/MF_Sptilight_Commuication}.Result
   B=Multiply(A=Add(A=@n17) {constB=0.02}, B=Multiply(A=@n19, B=@n20.A))
  )
  n22 = Fresnel(
   ExponentIn=ScalarParameter() {defaultValue=2, parameterName=Exp}
   BaseReflectFractionIn=ScalarParameter() {defaultValue=0.8, parameterName=Base}
   Normal=VertexNormalWS()
  )
  n23 = Multiply(A=Power(Base=OneMinus(None=@n22))) {constB=20}
  Multiply(A=Multiply(A=Multiply(A=@n21) {constB=2}, B=Multiply(A=@n18.A, B=@n23)), B=OneMinus(None=VertexColor().R))
  n24 = Transform(None=VertexNormalWS()) {transformSourceType=TRANSFORMSOURCE_World, transformType=TRANSFORM_Tangent}
  n25 = AppendVector(A=ComponentMask(None=@n24) {r=true}, B=ComponentMask(None=@n24) {g=true})
  n26 = Add(
   A=ComponentMask(None=@n24) {b=true}
   B=CollectionParameter() {collection=&Lighting_MPC, parameterName=SpotStrip_01_ConeSize}
  ) {constB=0}
  Multiply(A=Transform(None=AppendVector(A=@n25, B=@n26)), B=OneMinus(None=@n19))
 ]
}
